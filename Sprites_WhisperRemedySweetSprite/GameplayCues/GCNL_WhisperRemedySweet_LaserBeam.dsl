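; Looping gameplay cue notify for the WhisperRemedySweet sprite's laser beam, adding a Fortnite sound indicator on top of the parent cue's behavior.
; On loop start, if a scalable-float toggle is enabled and the target pawn is not locally viewed, it starts a looping indicator at the instigator's location using the laser loop sound.
; On removal, the same toggle gates stopping the indicator, without the viewer check.
; Both paths print debug strings.
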(event EventOnLoopingStart(EitherVFX-preferthisversion)
  (params (MyTarget :type Actor%20Object%20Reference) (Parameters :type Gameplay%20Cue%20Parameters%20Structure%20%28by%20ref%29) (ParticleComponents :type Array%20of%20FXSystem%20Component%20Object%20References) (AudioComponents :type Array%20of%20Audio%20Component%20Object%20References))
  (|CallParentFunction MyTarget Parameters ParticleComponents AudioComponents)
  (if (ScalableFloat|GetValueasBool (Variables|Default|GetShouldEnableGCAVI))
    (bind _asfort_pawn (Utilities|Casting|CastToFortPawn MyTarget))
    (if (not (Pawn|IsLocallyViewed _asfort_pawn))
      (SoundIndicator|StartSoundIndicatorLooping (Variables|Default|GetFortSoundIndicator) (Transformation|GetActorLocation (Ability|GameplayCue|GetInstigatorActor Parameters)) "/Sprites_WhisperRemedySweetSprite/Assets/Sounds/Cues/SFX/Laser_Eyes/IP_SC_WhisperRemedySweet_Laser_Loop_VO.IP_SC_WhisperRemedySweet_Laser_Loop_VO" 1.0 0)
      (Development|PrintString 0 "START INDICATOR LOOPING !!!"))))

(event EventOnRemoval(EitherVFX-preferthisversion)
  (params (MyTarget :type Actor%20Object%20Reference) (Parameters :type Gameplay%20Cue%20Parameters%20Structure%20%28by%20ref%29) (ParticleComponents :type Array%20of%20FXSystem%20Component%20Object%20References) (AudioComponents :type Array%20of%20Audio%20Component%20Object%20References) (BurstCameraShakeInstance :type Camera%20Shake%20Base%20Object%20Reference) (BurstDecalInstance :type Decal%20Actor%20Object%20Reference))
  (|CallParentFunction MyTarget Parameters ParticleComponents AudioComponents BurstCameraShakeInstance BurstDecalInstance)
  (if (ScalableFloat|GetValueasBool (Variables|Default|GetShouldEnableGCAVI))
    (SoundIndicator|StopSoundIndicatorLooping (Variables|Default|GetFortSoundIndicator))
    (Development|PrintString 0 "STOPPING INDICATOR LOOPING !!!")))

(fn ConstructionScript)
